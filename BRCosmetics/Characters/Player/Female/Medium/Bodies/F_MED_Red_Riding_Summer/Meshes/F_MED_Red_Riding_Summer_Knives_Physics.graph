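# Physics asset for the knives mesh of the F_MED_Red_Riding_Summer female player cosmetic, covering its left and right dyn_kinfe bones.
# Each side has two capsule bodies: a wide one on the base bone and a slimmer, longer one on the parent bone.
# One constraint per side links that pair with tight swing and twist angles, and no constraint joins left to right.
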PhysicsAsset /BRCosmetics/Characters/Player/Female/Medium/Bodies/F_MED_Red_Riding_Summer/Meshes/F_MED_Red_Riding_Summer_Knives_Physics
bodies {
 dyn_kinfe_l_base [{shapeType=Capsule, length=12.875, radius=7.875, center=<-0.51 -1.3278 8.5896>, rotation=<rot 78.2323 -20.2543 0.0001>}]
 dyn_kinfe_l_parent [{shapeType=Capsule, length=18.1645, radius=3.5353, center=<0 0 -1.0827>, rotation=<rot 0 -99.5175 -87.2349>}]
 dyn_kinfe_r_base [{shapeType=Capsule, length=12.5, radius=8.25, center=<-0.0092 0.5511 8.3647>, rotation=<rot 86.6107 27.4298 0>}]
 dyn_kinfe_r_parent [{shapeType=Capsule, length=18.1645, radius=3.4743, center=<0 0 -1.432>, rotation=<rot 0 97.9258 -89.2174>}]
}
constraints [
 {bone1Name=dyn_kinfe_l_parent, bone2Name=dyn_kinfe_l_base, swing1LimitDegrees=10, swing2LimitDegrees=5, twistLimitDegrees=2}
 {bone1Name=dyn_kinfe_r_parent, bone2Name=dyn_kinfe_r_base, swing1LimitDegrees=10, swing2LimitDegrees=5, twistLimitDegrees=2}
]
