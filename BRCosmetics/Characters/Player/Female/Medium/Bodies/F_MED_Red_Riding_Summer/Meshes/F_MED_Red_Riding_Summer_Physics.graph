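# Physics asset for the F_MED_Red_Riding_Summer female body cosmetic, covering only the dynamic shoes and shoelaces plus nearby leg bones.
# Box shoe bodies are constrained to small calf-twist capsules, and inner and outer lace capsules to per-side parent boxes, with narrow twist and wider swing limits.
# Foot capsules and lace base boxes have no constraints, and no constraint links the lace parents to the shoes.
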PhysicsAsset /BRCosmetics/Characters/Player/Female/Medium/Bodies/F_MED_Red_Riding_Summer/Meshes/F_MED_Red_Riding_Summer_Physics
bodies {
 calf_twist_01_l [{shapeType=Capsule, length=0, radius=0.75, center=<0 -1.3367 0>, rotation=<rot 90 0 0>}]
 calf_twist_01_r [{shapeType=Capsule, length=0, radius=0.75, center=<0.0083 1.3289 0.1438>, rotation=<rot 90 0 0>}]
 dyn_shoe_l [{shapeType=Box, extentX=8.2261, extentY=8.138, extentZ=2.748, center=<0 0 1.0965>}]
 dyn_shoe_r [{shapeType=Box, extentX=8.2261, extentY=8.138, extentZ=2.748, center=<0 0 1.0965>}]
 dyn_shoelace_l_base [{shapeType=Box, extentX=0.5, extentY=0.5, extentZ=0.5}]
 dyn_shoelace_l_inner [{shapeType=Capsule, length=2.2102, radius=1.3666, center=<2.2788 0 0.0512>, rotation=<rot 89.8703 -179.9999 -179.9998>}]
 dyn_shoelace_l_outter [{shapeType=Capsule, length=1.0852, radius=1.5506, center=<3.1875 -0.3467 0.3079>, rotation=<rot -84.4605 -0.004 -13.3265>}]
 dyn_shoelace_l_parent [{shapeType=Box, extentX=0.5, extentY=0.5, extentZ=0.5}]
 dyn_shoelace_r_base [{shapeType=Box, extentX=0.5, extentY=0.5, extentZ=0.5}]
 dyn_shoelace_r_inner [{shapeType=Capsule, length=2.8352, radius=1.3006, center=<2.3925 0.4237 0.2308>, rotation=<rot -84.4603 -0.0093 14.8037>}]
 dyn_shoelace_r_outter [{shapeType=Capsule, length=3.2207, radius=1.0169, center=<2.366 0 0.1004>, rotation=<rot -87.5707 0 0>}]
 dyn_shoelace_r_parent [{shapeType=Box, extentX=0.5, extentY=0.5, extentZ=0.5}]
 foot_l [{shapeType=Capsule, length=11.5, radius=3.125, center=<-2.2199 3.2325 0.16>, rotation=<rot 0 20.3968 89.9995>}]
 foot_r [{shapeType=Capsule, length=11.5, radius=3.125, center=<2.1762 -3.5425 0.1602>, rotation=<rot 0 20.397 89.9996>}]
}
constraints [
 {bone1Name=dyn_shoelace_l_inner, bone2Name=dyn_shoelace_l_parent, swing1LimitDegrees=20, swing2LimitDegrees=20, twistLimitDegrees=3}
 {bone1Name=dyn_shoelace_l_outter, bone2Name=dyn_shoelace_l_parent, swing1LimitDegrees=30, swing2LimitDegrees=30, twistLimitDegrees=3}
 {bone1Name=dyn_shoelace_r_outter, bone2Name=dyn_shoelace_r_parent, swing1LimitDegrees=20, swing2LimitDegrees=20, twistLimitDegrees=3}
 {bone1Name=dyn_shoelace_r_inner, bone2Name=dyn_shoelace_r_parent, swing1LimitDegrees=20, swing2LimitDegrees=20, twistLimitDegrees=3}
 {bone1Name=dyn_shoe_r, bone2Name=calf_twist_01_r, swing1LimitDegrees=40, swing2LimitDegrees=20, twistLimitDegrees=5}
 {bone1Name=dyn_shoe_l, bone2Name=calf_twist_01_l, swing1LimitDegrees=40, swing2LimitDegrees=20, twistLimitDegrees=5}
]
