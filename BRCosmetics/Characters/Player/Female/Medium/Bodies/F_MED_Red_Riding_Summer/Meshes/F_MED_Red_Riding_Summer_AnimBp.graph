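# Animation Blueprint for the F_MED_Red_Riding_Summer character body, adding secondary physics motion to a pose copied from its attached parent mesh.
# The event graph maps each thigh's swing angle from ref pose to a per-leg alpha, which drives additive scaling of the dyn_sub_thigh bones.
# Two chained RigidBody nodes then simulate the knives and main physics assets in root-bone space, with overridden gravity and damped component motion.
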AnimBlueprint /BRCosmetics/Characters/Player/Female/Medium/Bodies/F_MED_Red_Riding_Summer/Meshes/F_MED_Red_Riding_Summer_AnimBp
target_skeleton=&/F_MED_Red_Riding_Summer_Skeleton
graphs [
 AnimGraph {nodes [
   n1 = "Transform (Modify) Bone - Bone: dyn_sub_thigh_r"(
    ComponentPose=Local To Component(LocalPose=Copy Pose From Mesh() {bUseAttachedParent=true})
    Alpha=Get legSwingDelta_r() -> (legSwingDelta_r:double)
    Translation="0,0,0"
    Rotation="0,0,0"
    Scale="6,6,6"
   ) {
    scaleMode=BMM_Additive
    scaleSpace=BCS_BoneSpace
    alphaBoolBlend {blendOption=Linear}
    boneToModify {boneName=dyn_sub_thigh_r}
    scale=<6 6 6>
   }
   n2 = "Transform (Modify) Bone - Bone: dyn_sub_thigh_l"(ComponentPose=@n1, Alpha=Get legSwingDelta_l() -> (legSwingDelta_l:double), Translation="0,0,0", Rotation="0,0,0", Scale="6,6,6") {
    scaleMode=BMM_Additive
    scaleSpace=BCS_BoneSpace
    alphaBoolBlend {blendOption=Linear}
    boneToModify {boneName=dyn_sub_thigh_l}
    scale=<6 6 6>
   }
   n3 = RigidBody(ComponentPose=@n2) {
    bClampLinearTranslationLimitToRefPose=true
    bOverrideWorldGravity=true
    overridePhysicsAsset=&/F_MED_Red_Riding_Summer_Knives_Physics
    simulationSpace=BaseBoneSpace
    alphaBoolBlend {blendOption=Linear}
    baseBoneRef {boneName=root}
    componentAppliedLinearAccClamp=<6000 6000 6000>
    componentLinearAccScale=<0.25 0.25 0.25>
    componentLinearVelScale=<0.25 0.25 0.25>
    overrideWorldGravity {z=-980}
   }
   n4 = RigidBody(ComponentPose=@n3) {
    bClampLinearTranslationLimitToRefPose=true
    bOverrideWorldGravity=true
    overridePhysicsAsset=&/F_MED_Red_Riding_Summer_Physics
    simulationSpace=BaseBoneSpace
    alphaBoolBlend {blendOption=Linear}
    baseBoneRef {boneName=root}
    componentAppliedLinearAccClamp=<6000 6000 6000>
    componentLinearAccScale=<0.15 0.15 0.15>
    componentLinearVelScale=<0.15 0.15 0.15>
    overrideWorldGravity {z=-980}
   }
   Output Pose(Result=Component To Local(ComponentPose=@n4)) {layerGroup=None}
  ]}
 EventGraph {nodes [
   Event BlueprintUpdateAnimation()
   n1 = GetOwnerSkeletalMeshComponent()
   > Set legSwingDelta_r(legSwingDelta_r:double=MapRangeClamped(Value=GetTwistAndSwingAngleOfDeltaRotationFromRefPose(self=@n1, BoneName=thigh_r).OutSwingAngle, InRangeA=150, InRangeB=75)) -> (Output_Get:double)
   > Set legSwingDelta_l(legSwingDelta_l:double=MapRangeClamped(Value=GetTwistAndSwingAngleOfDeltaRotationFromRefPose(self=@n1, BoneName=thigh_l).OutSwingAngle, InRangeA=43, InRangeB=110)) -> (Output_Get:double)
   TryGetPawnOwner()
  ]}
]
variables [
 legSwingDelta_l {type=number}
 legSwingDelta_r {type=number}
]
