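# Physics asset for the braid and string bones of the Red Riding Summer female medium face accessory, providing collision bodies and joints for their secondary motion.
# Left and right braids are four-capsule chains hanging from spine_05, with free swing and twist.
# Each string chains from a small base box through parent, 01 and 02 bodies, using tighter angular limits that differ slightly by side.
# The spine_05 body combines capsules and a box as an upper-torso collision volume.
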PhysicsAsset /BRCosmetics/Characters/Player/Female/Medium/Bodies/F_MED_Red_Riding_Summer/Meshes/Parts/F_MED_Red_Riding_Summer_FaceAcc_Strings_Physics
bodies {
 dyn_braid_l_01 [{shapeType=Capsule, length=5.4496, radius=2.8106, shapeName=dyn_braid_l_01_capsule, center=<2.709 -0.3292 0.4308>, rotation=<rot 90 180 180>}]
 dyn_braid_l_02 [{shapeType=Capsule, length=3.7443, radius=2.5007, shapeName=dyn_braid_l_02_capsule, center=<4.5817 -0.0736 0.5084>, rotation=<rot 90 180 180>}]
 dyn_braid_l_03 [{shapeType=Capsule, length=2.2786, radius=2.0085, shapeName=dyn_braid_l_03_capsule, center=<3.831 0.3026 0.3463>, rotation=<rot 90 180 180>}]
 dyn_braid_l_04 [{shapeType=Capsule, length=4.3794, radius=1.8557, shapeName=dyn_braid_l_04_capsule, center=<4.2264 0.2642 0.4044>, rotation=<rot 90 180 180>}]
 dyn_braid_r_01 [{shapeType=Capsule, length=4.0421, radius=2.722, shapeName=dyn_braid_r_01_capsule, center=<2.3212 0.4886 0.3575>, rotation=<rot 90 180 180>}]
 dyn_braid_r_02 [{shapeType=Capsule, length=3.5398, radius=1.8709, shapeName=dyn_braid_r_02_capsule, center=<3.5568 0.4032 0.1959>, rotation=<rot 90 180 180>}]
 dyn_braid_r_03 [{shapeType=Capsule, length=3.4121, radius=1.4004, shapeName=dyn_braid_r_03_capsule, center=<3.5924 0.0427 0.1093>, rotation=<rot 90 180 180>}]
 dyn_braid_r_04 [{shapeType=Capsule, length=3.7136, radius=1.944, shapeName=dyn_braid_r_04_capsule, center=<3.3472 -0.3333 0.583>, rotation=<rot 90 0 0>}]
 dyn_string_l_01 [{shapeType=Capsule, length=0.505, radius=0.755, shapeName=dyn_string_l_01_capsule, center=<0.9593 -0.0473 -0.7776>, rotation=<rot 90 0 0>}]
 dyn_string_l_02 [{shapeType=Capsule, length=0.505, radius=0.755, shapeName=dyn_string_l_02_capsule, center=<0.9813 -0.2515 -0.7198>, rotation=<rot 90 0 0>}]
 dyn_string_l_base [{shapeType=Box, extentX=0.5, extentY=0.5, extentZ=0.5}]
 dyn_string_l_parent [{shapeType=Sphere, radius=0.5, shapeName=dyn_string_r_parent_sphere}]
 dyn_string_r_01 [{shapeType=Capsule, length=0.505, radius=0.755, shapeName=dyn_string_r_01_capsule, center=<1.5884 0.0031 -0.6771>, rotation=<rot 90 0 0>}]
 dyn_string_r_02 [{shapeType=Capsule, length=0.505, radius=0.755, shapeName=dyn_string_r_02_capsule, center=<1.1995 -0.0107 -0.6012>, rotation=<rot 90 0 0>}]
 dyn_string_r_base [{shapeType=Box, extentX=0.5, extentY=0.5, extentZ=0.5}]
 dyn_string_r_parent [{shapeType=Capsule, length=2.0431, radius=0.75, shapeName=dyn_string_r_parent_capsule, center=<1.2797 -0.0887 -0.6623>, rotation=<rot 90 0 0>}]
 spine_05 [
  {shapeType=Capsule, length=20.75, radius=9.25, center=<-8.1193 2.7821 0>}
  {shapeType=Capsule, length=20.375, radius=7.875, center=<4.7634 -1.4486 0>}
  {shapeType=Capsule, length=9.5, radius=15, shapeName=spine_05_capsule, center=<-9.3551 -1.2815 0>}
  {shapeType=Box, extentX=2.5, extentY=2.5, extentZ=2.5}
 ]
}
constraints [
 {bone1Name=dyn_string_r_parent, bone2Name=dyn_string_r_base, swing1LimitDegrees=30, swing2LimitDegrees=60, twistLimitDegrees=10}
 {bone1Name=dyn_string_r_01, bone2Name=dyn_string_r_parent, swing1LimitDegrees=30, swing2LimitDegrees=30, twistLimitDegrees=10}
 {bone1Name=dyn_string_r_02, bone2Name=dyn_string_r_01, swing1LimitDegrees=30, swing2LimitDegrees=30, twistLimitDegrees=10}
 {bone1Name=dyn_string_l_parent, bone2Name=dyn_string_l_base, swing1LimitDegrees=30, swing2LimitDegrees=30, twistLimitDegrees=5}
 {bone1Name=dyn_string_l_01, bone2Name=dyn_string_l_parent, swing1LimitDegrees=30, swing2LimitDegrees=30, twistLimitDegrees=5}
 {bone1Name=dyn_string_l_02, bone2Name=dyn_string_l_01, swing1LimitDegrees=30, swing2LimitDegrees=30, twistLimitDegrees=5}
 {
  bone1Name=dyn_braid_r_01
  bone2Name=spine_05
  swing1LimitDegrees=45
  swing1Motion=Free
  swing2LimitDegrees=45
  swing2Motion=Free
  twistLimitDegrees=45
  twistMotion=Free
 }
 {
  bone1Name=dyn_braid_r_02
  bone2Name=dyn_braid_r_01
  swing1LimitDegrees=45
  swing1Motion=Free
  swing2LimitDegrees=45
  swing2Motion=Free
  twistLimitDegrees=45
  twistMotion=Free
 }
 {
  bone1Name=dyn_braid_r_03
  bone2Name=dyn_braid_r_02
  swing1LimitDegrees=45
  swing1Motion=Free
  swing2LimitDegrees=45
  swing2Motion=Free
  twistLimitDegrees=45
  twistMotion=Free
 }
 {
  bone1Name=dyn_braid_r_04
  bone2Name=dyn_braid_r_03
  swing1LimitDegrees=45
  swing1Motion=Free
  swing2LimitDegrees=45
  swing2Motion=Free
  twistLimitDegrees=45
  twistMotion=Free
 }
 {
  bone1Name=dyn_braid_l_01
  bone2Name=spine_05
  swing1LimitDegrees=45
  swing1Motion=Free
  swing2LimitDegrees=45
  swing2Motion=Free
  twistLimitDegrees=45
  twistMotion=Free
 }
 {
  bone1Name=dyn_braid_l_02
  bone2Name=dyn_braid_l_01
  swing1LimitDegrees=45
  swing1Motion=Free
  swing2LimitDegrees=45
  swing2Motion=Free
  twistLimitDegrees=45
  twistMotion=Free
 }
 {
  bone1Name=dyn_braid_l_03
  bone2Name=dyn_braid_l_02
  swing1LimitDegrees=45
  swing1Motion=Free
  swing2LimitDegrees=45
  swing2Motion=Free
  twistLimitDegrees=45
  twistMotion=Free
 }
 {
  bone1Name=dyn_braid_l_04
  bone2Name=dyn_braid_l_03
  swing1LimitDegrees=45
  swing1Motion=Free
  swing2LimitDegrees=45
  swing2Motion=Free
  twistLimitDegrees=45
  twistMotion=Free
 }
]
