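# Physics asset for the hood and face-accessory part of the female medium Red Riding Summer character, simulating two braids and the hood.
# Each braid is a capsule chain from a small base sphere, with limited swing and twist joints and swing1 locked near the root.
# The hood is a box constrained to its base body.
# Unconstrained head, neck, clavicle and upper-spine shapes appear to serve as colliders for the dangling parts.
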PhysicsAsset /BRCosmetics/Characters/Player/Female/Medium/Bodies/F_MED_Red_Riding_Summer/Meshes/Parts/F_MED_Red_Riding_Summer_Hood_FaceAcc_Physics
bodies {
 clavicle_l [{shapeType=Capsule, length=22, radius=8, center=<5.5442 3.8413 -9.7907>, rotation=<rot 2.1447 -1.8097 -42.2377>}]
 clavicle_r [{shapeType=Capsule, length=22, radius=8, center=<-4.7788 -4.6545 9.6888>, rotation=<rot 0 0 -39.734>}]
 dyn_braid_l_01 [{shapeType=Capsule, length=2.5353, radius=2.0553, center=<3.1443 0.0001 0.0731>, rotation=<rot 90 0 0>}]
 dyn_braid_l_02 [{shapeType=Capsule, length=2.5353, radius=2.0553, center=<3.1443 0.0001 0.0731>, rotation=<rot 90 0 0>}]
 dyn_braid_l_03 [{shapeType=Capsule, length=2.9977, radius=1.9376, center=<3.0795 0.0001 0>, rotation=<rot 90 0 0>}]
 dyn_braid_l_04 [{shapeType=Capsule, length=4.133, radius=1.8384, center=<4.048 0 0>, rotation=<rot 90 0 0>}]
 dyn_braid_l_base [{shapeType=Sphere, radius=0.625}]
 dyn_braid_l_parent [{shapeType=Capsule, length=2.5353, radius=2.0553, center=<3.1443 0.0001 0.0731>, rotation=<rot 90 0 0>}]
 dyn_braid_r_01 [{shapeType=Capsule, length=2.5353, radius=2.0553, center=<3.1443 0.0001 0.0731>, rotation=<rot 90 0 0>}]
 dyn_braid_r_02 [{shapeType=Capsule, length=2.7853, radius=1.8053, center=<3.1443 0.0001 0.0731>, rotation=<rot 90 0 0>}]
 dyn_braid_r_03 [{shapeType=Capsule, length=3.2477, radius=1.6876, center=<3.1022 -0.2657 0.2041>, rotation=<rot 90 0 0>}]
 dyn_braid_r_04 [{shapeType=Capsule, length=4.0664, radius=1.9371, center=<3.8685 0 0>, rotation=<rot 90 0 0>}]
 dyn_braid_r_base [{shapeType=Sphere, radius=0.5}]
 dyn_braid_r_parent [{shapeType=Capsule, length=2.5353, radius=2.0553, center=<3.1443 0.0001 0.0731>, rotation=<rot 90 0 0>}]
 dyn_hood_base [{shapeType=Box, extentX=0.5, extentY=0.5, extentZ=0.5}]
 dyn_hood_parent [{shapeType=Box, extentX=6.9601, extentY=7.134, extentZ=2.1728, center=<3.5117 0 -0.587>, rotation=<rot -11.9494 0 0>}]
 head [{shapeType=Capsule, length=2.875, radius=7.375, center=<5.8026 2.998 1.0072>, rotation=<rot 90 0 0>}]
 neck_02 [{shapeType=Capsule, length=6.375, radius=3.875, rotation=<rot 90 0 0>}]
 spine_05 [
  {shapeType=Capsule, length=20.75, radius=9.25, center=<-8.4452 3.3246 0>}
  {shapeType=Capsule, length=20.375, radius=7.875, center=<3.3572 -1.2104 0>}
  {shapeType=Box, extentX=2.5, extentY=2.5, extentZ=2.5}
 ]
}
constraints [
 {
  bone1Name=dyn_braid_r_parent
  bone2Name=dyn_braid_r_base
  swing1LimitDegrees=15
  swing1Motion=Locked
  swing2LimitDegrees=60
  twistLimitDegrees=3
 }
 {
  bone1Name=dyn_braid_r_01
  bone2Name=dyn_braid_r_parent
  swing1LimitDegrees=15
  swing1Motion=Locked
  swing2LimitDegrees=50
  twistLimitDegrees=3
 }
 {bone1Name=dyn_braid_r_02, bone2Name=dyn_braid_r_01, swing1LimitDegrees=15, swing2LimitDegrees=30, twistLimitDegrees=3}
 {bone1Name=dyn_braid_r_03, bone2Name=dyn_braid_r_02, swing1LimitDegrees=15, swing2LimitDegrees=30, twistLimitDegrees=3}
 {
  bone1Name=dyn_braid_l_parent
  bone2Name=dyn_braid_l_base
  swing1LimitDegrees=15
  swing1Motion=Locked
  swing2LimitDegrees=60
  twistLimitDegrees=5
 }
 {
  bone1Name=dyn_braid_l_01
  bone2Name=dyn_braid_l_parent
  swing1LimitDegrees=15
  swing1Motion=Locked
  swing2LimitDegrees=50
  twistLimitDegrees=3
 }
 {bone1Name=dyn_braid_l_02, bone2Name=dyn_braid_l_01, swing1LimitDegrees=15, swing2LimitDegrees=30, twistLimitDegrees=3}
 {bone1Name=dyn_braid_l_03, bone2Name=dyn_braid_l_02, swing1LimitDegrees=15, swing2LimitDegrees=30, twistLimitDegrees=5}
 {bone1Name=dyn_hood_parent, bone2Name=dyn_hood_base, swing1LimitDegrees=30, swing2LimitDegrees=40, twistLimitDegrees=1}
 {bone1Name=dyn_braid_l_04, bone2Name=dyn_braid_l_03, swing1LimitDegrees=20, swing2LimitDegrees=20, twistLimitDegrees=3}
 {bone1Name=dyn_braid_r_04, bone2Name=dyn_braid_r_03, swing1LimitDegrees=20, swing2LimitDegrees=20, twistLimitDegrees=3}
]
