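# Post-copy AnimBlueprint for the Red Riding Summer face accessory, adding secondary motion to braids, bangs and strings.
# The parent's pose gets targeting-driven braid offsets, AnimDynamics bangs and RigidBody strings, with gravity and offset mapped from frame delta time, then is cached.
# The Stage2 cosmetic tag, checked on customization load, selects braid LookAts plus hood physics over standard face-accessory physics.
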AnimBlueprint /BRCosmetics/Characters/Player/Female/Medium/Bodies/F_MED_Red_Riding_Summer/Meshes/Parts/F_MED_Red_Riding_Summer_FaceAcc_AnimBp
target_skeleton=&/F_MED_Red_Riding_Summer_FaceAcc_Skeleton
graphs [
 AnimGraph {nodes [
   "Transform (Modify) Bone - Bone: dyn_braid_l_base"(Alpha=1, Translation="0,0,0", Rotation="0,-20,0.000", Scale="1,1,1") {
    rotationMode=BMM_Additive
    rotationSpace=BCS_BoneSpace
    translationMode=BMM_Additive
    translationSpace=BCS_BoneSpace
    alphaBoolBlend {blendOption=Linear}
    alphaScaleBiasClamp {bInterpResult=true, interpSpeedDecreasing=4, interpSpeedIncreasing=6}
    boneToModify {boneName=dyn_braid_l_base}
    rotation {yaw=-20}
   }
   "Transform (Modify) Bone - Bone: dyn_braid_l_parent"(Alpha=1, Translation="0,0,0", Rotation="0,-20,0.000", Scale="0.5,0.5,0.5") {
    rotationMode=BMM_Additive
    rotationSpace=BCS_BoneSpace
    translationMode=BMM_Additive
    translationSpace=BCS_BoneSpace
    alphaBoolBlend {blendOption=Linear}
    alphaScaleBiasClamp {bInterpResult=true, interpSpeedDecreasing=4, interpSpeedIncreasing=6}
    boneToModify {boneName=dyn_braid_l_parent}
    rotation {yaw=-20}
    scale=<0.5 0.5 0.5>
   }
   RigidBody() {
    bClampLinearTranslationLimitToRefPose=true
    bOverrideWorldGravity=true
    overridePhysicsAsset=&/F_MED_Red_Riding_Summer_FaceAcc_Strings_Physics
    simulationSpace=BaseBoneSpace
    alphaBoolBlend {blendOption=Linear}
    alphaScaleBiasClamp {bInterpResult=true, interpSpeedDecreasing=8, interpSpeedIncreasing=4}
    baseBoneRef {boneName=spine_02}
    componentAppliedLinearAccClamp=<6000 6000 6000>
    componentLinearAccScale=<0.25 0.25 0.25>
    componentLinearVelScale=<0.25 0.25 0.25>
    overrideWorldGravity {z=-980}
   }
   RigidBody() {
    bClampLinearTranslationLimitToRefPose=true
    bOverrideWorldGravity=true
    overridePhysicsAsset=&F_MED_Red_Riding_Summer_FaceAcc_Strings_Physics
    simulationSpace=BaseBoneSpace
    alphaBoolBlend {blendOption=Linear}
    alphaScaleBiasClamp {bInterpResult=true, interpSpeedDecreasing=8, interpSpeedIncreasing=4}
    baseBoneRef {boneName=spine_02}
    componentAppliedLinearAccClamp=<6000 6000 6000>
    componentLinearAccScale=<0.25 0.25 0.25>
    componentLinearVelScale=<0.25 0.25 0.25>
    overrideWorldGravity {z=-980}
   }
   n1 = RigidBody(ComponentPose=Local To Component(LocalPose=Use cached pose 'HairCache'())) {
    bClampLinearTranslationLimitToRefPose=true
    bOverrideWorldGravity=true
    overridePhysicsAsset=&/F_MED_Red_Riding_Summer_FaceAcc_Physics
    simulationSpace=BaseBoneSpace
    alphaBoolBlend {blendOption=Linear}
    alphaScaleBiasClamp {bInterpResult=true, interpSpeedDecreasing=4, interpSpeedIncreasing=4}
    baseBoneRef {boneName=spine_05}
    componentAppliedLinearAccClamp=<6000 6000 6000>
    componentLinearAccScale=<0.25 0.25 0.25>
    componentLinearVelScale=<0.25 0.25 0.25>
    overrideWorldGravity {z=-980}
   }
   n2 = Look At - Bone: dyn_braid_r_parent(ComponentPose=Local To Component(LocalPose=Use cached pose 'HairCache'()), Alpha=1) {
    lookAtClamp=100
    alphaBoolBlend {blendOption=Linear}
    boneToModify {boneName=dyn_braid_r_parent}
    lookAtTarget {boneReference {boneName=clavicle_r}}
   }
   n3 = Look At - Bone: dyn_braid_l_parent(ComponentPose=@n2, Alpha=1) {
    lookAtClamp=100
    alphaBoolBlend {blendOption=Linear}
    boneToModify {boneName=dyn_braid_l_parent}
    lookAtTarget {boneReference {boneName=clavicle_l}}
   }
   n4 = RigidBody(ComponentPose=@n3) {
    bClampLinearTranslationLimitToRefPose=true
    bOverrideWorldGravity=true
    overridePhysicsAsset=&/F_MED_Red_Riding_Summer_Hood_FaceAcc_Physics
    simulationSpace=BaseBoneSpace
    alphaBoolBlend {blendOption=Linear}
    alphaScaleBiasClamp {bInterpResult=true, interpSpeedDecreasing=4, interpSpeedIncreasing=4}
    baseBoneRef {boneName=spine_05}
    componentAppliedLinearAccClamp=<6000 6000 6000>
    componentLinearAccScale=<0.25 0.25 0.25>
    componentLinearVelScale=<0.25 0.25 0.25>
    overrideWorldGravity {z=-980}
   }
   Output Pose(Result=Blend Poses by bool(
     BlendPose_0=Component To Local(ComponentPose=@n4)
     BlendPose_1=Component To Local(ComponentPose=@n1)
     bActiveValue=Get hair_base_active() -> (hair_base_active:bool)
    ) {blendType=Linear}) {layerGroup=None}
   n5 = Get ADS_Alpha() -> (ADS_Alpha:double)
   n6 = "Transform (Modify) Bone - Bone: dyn_braid_r_base"(
    ComponentPose=Local To Component(LocalPose=Copy Pose From Mesh() {bUseAttachedParent=true})
    Alpha=@n5
    Translation="0,1,0"
    Rotation="0.000,15,0.000"
    Scale="1,1,1"
   ) {
    rotationMode=BMM_Additive
    rotationSpace=BCS_BoneSpace
    translationMode=BMM_Additive
    translationSpace=BCS_BoneSpace
    alphaBoolBlend {blendOption=Linear}
    alphaScaleBiasClamp {bInterpResult=true, interpSpeedDecreasing=4, interpSpeedIncreasing=6}
    boneToModify {boneName=dyn_braid_r_base}
    rotation {yaw=15}
    translation {y=1}
   }
   n7 = "Transform (Modify) Bone - Bone: dyn_braid_l_base"(ComponentPose=@n6, Alpha=@n5, Translation="0,1,0", Rotation="0.000,-15,0.000", Scale="1,1,1") {
    rotationMode=BMM_Additive
    rotationSpace=BCS_BoneSpace
    translationMode=BMM_Additive
    translationSpace=BCS_BoneSpace
    alphaBoolBlend {blendOption=Linear}
    alphaScaleBiasClamp {bInterpResult=true, interpSpeedDecreasing=4, interpSpeedIncreasing=6}
    boneToModify {boneName=dyn_braid_l_base}
    rotation {yaw=-15}
    translation {y=1}
   }
   n8 = Anim Dynamics - Bone: dyn_bang_parent(ComponentPose=@n7, Alpha=0.5, ExternalForce="0,0,0") {
    angularDampingOverride=0.9999
    bGravityOverrideInSimSpace=true
    bOverrideAngularDamping=true
    bOverrideLinearDamping=true
    linearDampingOverride=0.9999
    alphaBoolBlend {blendOption=Linear}
    boundBone {boneName=dyn_bang_parent}
    componentAppliedLinearAccClamp=<6000 6000 6000>
    componentLinearAccScale=<0.1 0.1 0.1>
    componentLinearVelScale=<0.1 0.1 0.1>
    physicsBodyDefinitions [{
      sphereCollisionRadius=0
      boundBone {boneName=dyn_bang_parent}
      boxExtents=<6 6 3>
      constraintSetup {angularLimitsMax=<0 35 20>, angularLimitsMin=<0 -5 -20>, angularTarget=<0 0 0>}
      localJointOffset=<2.25 0 0>
     }]
    relativeSpaceBone {boneName=spine_05}
   }
   n9 = "Transform (Modify) Bone - Bone: dyn_string_r_base"!(
    ComponentPose=@n8
    bAlphaBoolEnabled:bool=True
    Translation=Get Stripe_translation() -> (Stripe_translation:Vector)
    Rotation="0,0,0"
    Scale="1,1,1"
   ) -> (Pose) {
    alphaInputType=Bool
    rotationMode=BMM_Additive
    rotationSpace=BCS_BoneSpace
    translationMode=BMM_Additive
    translationSpace=BCS_BoneSpace
    alphaBoolBlend {blendOption=Linear}
    boneToModify {boneName=dyn_string_r_base}
    translation {y=0.23}
   }
   n10 = RigidBody!(ComponentPose=@n9, Alpha, OverrideWorldGravity:Vector=Get Straps_Rigid_Body_Alpha() -> (Straps_Rigid_Body_Alpha:Vector), ExternalForce) -> (Pose) {
    bClampLinearTranslationLimitToRefPose=true
    bOverrideWorldGravity=true
    overridePhysicsAsset=&F_MED_Red_Riding_Summer_FaceAcc_Strings_Physics
    simulationSpace=BaseBoneSpace
    alphaBoolBlend {blendOption=Linear}
    alphaScaleBiasClamp {bInterpResult=true, interpSpeedDecreasing=8, interpSpeedIncreasing=4}
    baseBoneRef {boneName=spine_02}
    componentAppliedLinearAccClamp=<6000 6000 6000>
    componentLinearAccScale=<0.25 0.25 0.25>
    componentLinearVelScale=<0.25 0.25 0.25>
    overrideWorldGravity {z=-500}
   }
   Save cached pose 'HairCache'(Pose=Component To Local(ComponentPose=@n10)) {cacheName=HairCache}
  ]}
 EventGraph {nodes [
   n1 = Event BlueprintUpdateAnimation()
   Set ADS_Alpha(execute=[@n1.then, @n3.then_0], ADS_Alpha:double=SelectFloat(A=1, B=0, bPickA=Get bIsTargeting())) -> (Output_Get:double)
   > Set Straps_Rigid_Body_Alpha(Straps_Rigid_Body_Alpha:Vector=MakeVector(Y=0, Z=MapRangeClamped(Value=@n1.DeltaTimeX, InRangeA=0.02, InRangeB=0.028, OutRangeA=-980, OutRangeB=-500))) -> (Output_Get:Vector)
   > Set Stripe_translation(Stripe_translation:Vector=MakeVector(Y=MapRangeClamped(Value=@n1.DeltaTimeX, InRangeA=0.02, InRangeB=0.028, OutRangeA=0.23, OutRangeB=0.6))) -> (Output_Get:Vector)
   Event OnCharacterCustomizationLoaded()
   > n2 = Branch(Condition=CharacterItemDefinitionHasAnyMetaTag(Tags=Make Array("[0]"=(TagName="Cosmetics.Variant.Property.Stage2"))))
   > Set hair_base_active(hair_base_active:bool=true) -> (Output_Get:bool)
   Set hair_base_active(execute=@n2.else, hair_base_active=false)
   Event BlueprintInitializeAnimation()
   > Set MeshToCopy(MeshToCopy:Skeletal Mesh Component Object Reference=GetOwnerSkeletalMeshComponent()) -> (Output_Get:Skeletal Mesh Component Object Reference)
   n3 = Sequence()
   n4 = Get PawnSpeed2D()
   Set Hair_Alpha(Hair_Alpha:double=SelectFloat(A=0.25, B=1, bPickA="float >= float"(A=@n4, B=1) {is=title K2Node_CallFunction})) -> (Output_Get:double)
  ]}
]
variables [
 ADS_Alpha {type=number}
 Hair_Alpha {type=number}
 hair_base_active {type=boolean}
 MeshToCopy {type=SkeletalMeshComponent}
 Straps_Rigid_Body_Alpha {type=Vector}
 Stripe_translation {type=Vector}
]
